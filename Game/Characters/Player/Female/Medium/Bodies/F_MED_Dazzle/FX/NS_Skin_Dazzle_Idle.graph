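# Idle skin effect for the Dazzle character that attaches animated, toon-shaded squiggle meshes to skeletal mesh sockets, with a separate emitter for the feet.
# Stepped ticks cycle shapes while time-based mesh indices and offset arrays fake flipbook jitter.
# System-level variant color arrays feed particle color, scaled by time of day.
# Skydiving shifts the main squiggles, TriggerKill switches foot squiggles to an extra frame set, and bisFootHidden scales them to zero.
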NiagaraSystem /Game/Characters/Player/Female/Medium/Bodies/F_MED_Dazzle/FX/NS_Skin_Dazzle_Idle
dependencies {
 usedDataInterfaces [
  &NiagaraDataInterfaceArrayFloat3
  &NiagaraDataInterfaceArrayInt32
  &NiagaraDataInterfaceParticleRead
  &NiagaraDataInterfaceArrayFloat
  &NiagaraDataInterfaceArrayColor
 ]
 usedDynamicInputs [&Boolean_NotOperation, &FNTimeOfDayScalar]
 usedModules [
  &self:SystemSpawnScript.SystemScriptSource.SystemScriptGraph.NiagaraNodeAssignment_3.SetVariables_#1
  &self:SystemSpawnScript.SystemScriptSource.SystemScriptGraph.NiagaraNodeAssignment_0.SetVariables_#2
  &self:SystemSpawnScript.SystemScriptSource.SystemScriptGraph.NiagaraNodeAssignment_4.SetVariables_#3
  &self:SystemSpawnScript.SystemScriptSource.SystemScriptGraph.NiagaraNodeAssignment_5.SetVariables_#4
  &self:SystemSpawnScript.SystemScriptSource.SystemScriptGraph.NiagaraNodeAssignment_6.SetVariables_#5
  &self:SystemSpawnScript.SystemScriptSource.SystemScriptGraph.NiagaraNodeAssignment_8.SetVariables_#6
  &self:SystemSpawnScript.SystemScriptSource.SystemScriptGraph.NiagaraNodeAssignment_9.SetVariables_#7
  &self:SystemSpawnScript.SystemScriptSource.SystemScriptGraph.NiagaraNodeAssignment_10.SetVariables_#8
  &self:SystemSpawnScript.SystemScriptSource.SystemScriptGraph.NiagaraNodeAssignment_7.SetVariables_#9
  &self:SystemSpawnScript.SystemScriptSource.SystemScriptGraph.NiagaraNodeAssignment_1.SetVariables_#10
  &SystemState
  &FN_Timer
  &self:SystemSpawnScript.SystemScriptSource.SystemScriptGraph.NiagaraNodeAssignment_2.SetVariables_#11
  &self:SetCurrentColorArray
  &Timeline
 ]
}
emitters {
 NE_Squiggles {
  emitterSpawnScript [
   /SetVariables(Emitter.FramePerShape=8, Emitter.ShapeNumber=5) {isSetParameters=true, module=SetVariables_#12}
   /SetVariables() {isSetParameters=true, module=SetVariables_#13}
  ]
  emitterUpdateScript [
   &EmitterState(Life Cycle Mode=::NewEnumerator1, Loop Behavior=::NewEnumerator1, Loop Duration Mode=::NewEnumerator1, UseLoopDelay=false)
   &SpawnBurst_Instantaneous(Spawn Count=2, Spawn Group=1, Use Spawn Probability=false, UseLoopCountLimit=false)
   /SetVariables(Emitter.SquiggleJitterSpeed=8) {isSetParameters=true}
   /Game/Effects/Niagara/MasterMaterials/ToonShader/Template/NM_ToonShaderParameters(HighLight_Softness=1, Lighting_Hardness=0, RimLight_Color:LinearColor=<rgba 0.1615 0.1615 0.1615 1>)
   /Game/Effects/Niagara/MasterMaterials/ToonShader/Template/NM_OutlineParameters()
  ]
  particleSpawnScript [
   &FN_SetSpawnGroup(Number of Spawn Groups=2)
   &InitializeParticle(
    Mesh Scale=&SelectVectorFromArray(
     Direct Array Index=&SpawnGroup()
     Vector Selection Array=DataInterface() {propertyValues {FloatData [<0.2 0.25 0.3>, <0.2 0.25 0.3>]}}
    )
    Mesh Scale Mode=::NewEnumerator4
    Sprite Size Mode=::NewEnumerator3
    Sprite UV Mode=::NewEnumerator3
    Uniform Sprite Size=25
    UsePositionOffset=false
    Write Material Random=true
   ) {module=InitializeParticle001}
   /SetVariables() {isSetParameters=true, module=SetVariables_#14}
   /SetVariables() {isSetParameters=true, module=SetVariables_#15}
   /SetVariables(Particles.ColorIndex:NiagaraInt32=&Multiply_Int(A=&ReturnExecIndex(), B=2)) {isSetParameters=true, module=SetVariables_#16}
   /SetVariables(Particles.CustomMeshIndex:NiagaraInt32=&RandomRangeInt(Minimum=0, Maximum=3)) {isSetParameters=true, module=SetVariables_#17}
   /SetVariables() {isSetParameters=true}
   &SkeletalMeshLocation(
    Bone / Socket Index=&SelectIntFromArray(Direct Array Index=&SpawnGroup(), Int Selection Array=DataInterface() {propertyValues {IntData [0,1]}})
    MaskSpawnGroup=false
    Mesh Sampling Type=::NewEnumerator5
    Sampled Orientation Coordinate Space=::World
    Skeletal Mesh=$User.SkeletalMesh
    Socket Sampling Mode=::NewEnumerator5
   ) {module=SkeletalMeshLocation001}
  ]
  particleUpdateScript [
   &ParticleState(Kill Particles When Lifetime Has Elapsed=false, Let Infinitely Lived Particles Die When Emitter Deactivates=true)
   &self:SteppedTickRate(FrameRefreshSpeed=0.2) {module=SteppedTickRate}
   &self:SkyDivingFloatModulator(IsSkydiving:NiagaraBool=$User.IsSkydiving, ModuloSpeed=3) {module=SkyDivingFloatModulator}
   /SetVariables(Particles.PositionOffset:Vector3f=&AddVectorToPosition(
     Position=&SelectVectorFromArray(
      Direct Array Index=&ModuloInt(
       A=&Add_Integer(
        A=&MakeIntFromFloat(Float=&Multiply_Float(A=$Engine.Time, B=$Emitter.SquiggleJitterSpeed))
        B=&ReturnExecIndex()
       )
       B=$Emitter.FramePerShape
      )
      Direct Array Mode=::NewEnumerator1
      Vector Selection Array=DataInterface() {propertyValues {FloatData [<-0.25 0 0>, <0.25 0 0>, <-0.5 0 0>, <0 0 0>, <0.25 0 0>, <-0.25 0 0>, <0.25 0 0>, <0 0 0>]}}
     )
     Vector=&Lerp_Vector(
      A=<0 0 0>
      B=&MaskVectorBySpawnGroup(Masked Vector=<5 0 2>, Passthrough Vector=<5 0 2>)
      Alpha=$Particles.SkyDivingModulator
     )
    )) {isSetParameters=true}
   &SkeletalMeshLocation(
    Bone / Socket Index=&SelectIntFromArray(Direct Array Index=&SpawnGroup(), Int Selection Array=DataInterface() {propertyValues {IntData [0,1]}})
    MaskSpawnGroup=false
    Mesh Sampling Type=::NewEnumerator5
    Sampled Orientation Coordinate Space=::World
    Sampled Position Offset=$Particles.PositionOffset
    Skeletal Mesh=$User.SkeletalMesh
    Socket Sampling Mode=::NewEnumerator5
    Velocity Sampling=::NewEnumerator0
    VelocitySpeedLimit=false
   )
   /SetVariables(Particles.MeshOrientation=&MultiplyQuaternion(
     Quaternion A=&MultiplyQuaternion(
      Quaternion A=$Particles.SkeletalMeshLocation.SampledRotation
      Quaternion B=&MakeQuatFromAxisAngle(
       Axis=<1 0 0>
       AngleInDegrees=&Lerp_Float(A=-20, B=20, Alpha=&MaskFloatBySpawnGroup(Masked Float=0, Passthrough Float=1))
      )
     )
     Quaternion B=&MakeQuatFromAxisAngle(Axis=<0 1 0>, AngleInDegrees=-45)
    )) {isSetParameters=true}
   /SetVariables(Particles.ShapeIndex:NiagaraInt32=&MakeCustomIntFromBool(
     Bool=$Particles.SteppedTickRate.RefreshTick
     True INT=&Add_Integer(A=$Particles.ShapeIndex, B=1)
     False INT=$Particles.ShapeIndex
    )) {isSetParameters=true, module=SetVariables_#18}
   &self:SetColorIndex(
    AttributeReader=DataInterface() {propertyValues {EmitterBinding {bindingMode=Other, emitterName=NE_Squiggles}}}
    NumberOfColors=5
    UsedColorArray:NiagaraDataInterfaceArrayInt32=$System.UsedColor
   ) {module=SetColorIndex}
   /SetVariables(Particles.CustomMeshIndex:NiagaraInt32=&MakeCustomIntFromBool(
     Bool=$User.TriggerKill
     True INT=&Add_Integer(
      A=&Multiply_Int(A=$Emitter.FramePerShape, B=$Emitter.ShapeNumber)
      B=&ModuloInt(
       A=&MakeIntFromFloat(Float=&Multiply_Float(A=$Engine.Time, B=$Emitter.SquiggleJitterSpeed))
       B=$Emitter.FramePerShape
      )
     )
     False INT=&ModuloInt(
      A=&Add_Integer(
       A=&ModuloInt(
        A=&MakeIntFromFloat(Float=&Multiply_Float(A=$Engine.Time, B=$Emitter.SquiggleJitterSpeed))
        B=$Emitter.FramePerShape
       )
       B=&Multiply_Int(
        A=&ModuloInt(
         A=&Add_Integer(A=$Particles.ShapeIndex, B=&ReturnExecIndex())
         B=$Emitter.ShapeNumber
        )
        B=$Emitter.FramePerShape
       )
      )
      B=&Multiply_Int(A=$Emitter.FramePerShape, B=$Emitter.ShapeNumber)
     )
    )) {enabled=false, isSetParameters=true, module=SetVariables_#19}
   /SetVariables(Particles.CustomMeshIndex:NiagaraInt32=&ModuloInt(
     A=&Add_Integer(
      A=&ModuloInt(
       A=&MakeIntFromFloat(Float=&Multiply_Float(A=$Engine.Time, B=$Emitter.SquiggleJitterSpeed))
       B=$Emitter.FramePerShape
      )
      B=&Multiply_Int(
       A=&ModuloInt(
        A=&Add_Integer(A=$Particles.ShapeIndex, B=&ReturnExecIndex())
        B=$Emitter.ShapeNumber
       )
       B=$Emitter.FramePerShape
      )
     )
     B=&Multiply_Int(A=$Emitter.FramePerShape, B=$Emitter.ShapeNumber)
    )) {isSetParameters=true}
   &ScaleColor(
    Color Value To Scale=$Particles.Color
    Scale RGB=&Multiply_VectorByFloat(Vector=<1 1 1>, Float=1)
    ScaleA=true
    ScaleRGB=true
   ) {module=ScaleColor001}
   &Color(Color=&MakeCustomLinearColorFromBool(
     Bool=$User.TriggerKill
     True Color=&SelectLinearColorFromArray(
      Array Sampling Mode=::NewEnumerator1
      Direct Array Index=$Particles.ColorIndex
      Direct Array Mode=::NewEnumerator1
      Color Selection Array=$System.ColorArrayCurrentVariant
      Random Seed=?
     )
     False Color=&SelectLinearColorFromArray(
      Array Sampling Mode=::NewEnumerator1
      Direct Array Index=$Particles.ColorIndex
      Direct Array Mode=::NewEnumerator1
      Color Selection Array=$System.ColorArrayCurrentVariant
      Random Seed=?
     )
    )) {module=Color001}
   /SetVariables(Particles.Color=&MultiplyLinearColor(
     Linear Color=$Particles.Color
     Scale Factor=&MakeVector4(X=$System.TimeOfDayBrightnessMultiplier, Y=$System.TimeOfDayBrightnessMultiplier, Z=$System.TimeOfDayBrightnessMultiplier, W=1)
    )) {isSetParameters=true}
  ]
  rendererClasses [&NiagaraMeshRendererProperties]
  renderers [
   {rendererClass=&NiagaraMeshRendererProperties, rendererIndex=0}
   {rendererClass=&NiagaraMeshRendererProperties, rendererIndex=1}
  ]
 }
 NE_Squiggles_Foot {
  emitterSpawnScript [
   /SetVariables(Emitter.FramePerShape=8, Emitter.ShapeNumber=4) {isSetParameters=true, module=SetVariables_#12}
   /SetVariables() {isSetParameters=true, module=SetVariables_#13}
  ]
  emitterUpdateScript [
   &EmitterState(Life Cycle Mode=::NewEnumerator1, Loop Behavior=::NewEnumerator1, Loop Duration Mode=::NewEnumerator1, UseLoopDelay=false)
   &SpawnBurst_Instantaneous(
    Spawn Count=&SelectIntFromArray(
     Direct Array Index=&MakeIntFromFloat(Float=$User.VariantIndex)
     Int Selection Array=DataInterface() {propertyValues {IntData [4,4,4,4,4,4]}}
    )
    Spawn Group=1
    Use Spawn Probability=false
    UseLoopCountLimit=false
   )
   /SetVariables(Emitter.SquiggleJitterSpeed=8) {isSetParameters=true}
   &NM_ToonShaderParameters(RimLight_Color:LinearColor=<rgba 0.1615 0.1615 0.1615 1>)
   &NM_OutlineParameters()
  ]
  particleSpawnScript [
   &FN_SetSpawnGroup()
   &InitializeParticle(
    Mesh Scale=<0.18 0.15 0.1>
    Mesh Scale Mode=::NewEnumerator4
    Sprite Size Mode=::NewEnumerator3
    Sprite UV Mode=::NewEnumerator3
    Uniform Sprite Size=25
    UsePositionOffset=false
    Write Material Random=true
   ) {module=InitializeParticle001}
   /SetVariables() {isSetParameters=true, module=SetVariables_#14}
   /SetVariables() {isSetParameters=true, module=SetVariables_#15}
   /SetVariables(Particles.ColorIndex:NiagaraInt32=&Multiply_Int(A=&ReturnExecIndex(), B=2)) {isSetParameters=true, module=SetVariables_#16}
   /SetVariables(Particles.CustomMeshIndex:NiagaraInt32=&RandomRangeInt(Minimum=0, Maximum=3)) {isSetParameters=true, module=SetVariables_#17}
   &SkeletalMeshLocation(
    Bone / Socket Index=&SelectIntFromArray(Direct Array Index=&SpawnGroup(), Int Selection Array=DataInterface() {propertyValues {IntData [2,2,3,3]}})
    MaskSpawnGroup=false
    Mesh Sampling Type=::NewEnumerator5
    Sampled Orientation Coordinate Space=::World
    Skeletal Mesh=$User.SkeletalMesh
    Socket Sampling Mode=::NewEnumerator5
   ) {module=SkeletalMeshLocation001}
  ]
  particleUpdateScript [
   &ParticleState(Kill Particles When Lifetime Has Elapsed=false, Let Infinitely Lived Particles Die When Emitter Deactivates=true)
   &self:SteppedTickRate(FrameRefreshSpeed=0.3) {module=SteppedTickRate}
   /SetVariables(Particles.PositionOffset:Vector3f=&Multiply_VectorByFloat(
     Vector=&SelectVectorFromArray(
      Direct Array Index=&SpawnGroup()
      Vector Selection Array=DataInterface() {propertyValues {FloatData [<1 0 0>, <1 0 0>, <0 0 1>, <0 0 1>, <0 0 1>, <0 0 1>]}}
     )
     Float=&SelectFloatFromArray(
      Direct Array Index=&ModuloInt(
       A=&Add_Integer(
        A=&MakeIntFromFloat(Float=&Multiply_Float(A=$Engine.Time, B=$Emitter.SquiggleJitterSpeed))
        B=&ReturnExecIndex()
       )
       B=$Emitter.FramePerShape
      )
      Direct Array Mode=::NewEnumerator1
      Float Selection Array=DataInterface() {propertyValues {FloatData [0.25,-0.5,0,-0.25,0.5,0.25,0,-0.25,0]}}
     )
    )) {isSetParameters=true}
   &SkeletalMeshLocation(
    Bone / Socket Index=&SelectIntFromArray(Direct Array Index=&SpawnGroup(), Int Selection Array=DataInterface() {propertyValues {IntData [2,2,3,3]}})
    MaskSpawnGroup=false
    Mesh Sampling Type=::NewEnumerator5
    Orientation Sampling=::NewEnumerator1
    Sampled Orientation Coordinate Space=::World
    Sampled Position Offset=&AddVectorToPosition(
     Position=&SelectVectorFromArray(
      Direct Array Index=&SpawnGroup()
      Vector Selection Array=DataInterface() {propertyValues {FloatData [<-1 0 -2>, <-7.5 0 1>, <-1 0 -2>, <-7.5 0 1>]}}
     )
     Vector=$Particles.PositionOffset
    )
    Skeletal Mesh=$User.SkeletalMesh
    Socket Sampling Mode=::NewEnumerator5
    Velocity Sampling=::NewEnumerator0
    VelocitySpeedLimit=false
   )
   /SetVariables(Particles.ShapeIndex:NiagaraInt32=&MakeCustomIntFromBool(
     Bool=$Particles.SteppedTickRate.RefreshTick
     True INT=&Add_Integer(A=$Particles.ShapeIndex, B=1)
     False INT=$Particles.ShapeIndex
    )) {isSetParameters=true, module=SetVariables_#18}
   &self:SetColorIndex(
    AttributeReader=DataInterface() {propertyValues {EmitterBinding {bindingMode=Other, emitterName=NE_Squiggles}}}
    NumberOfColors=5
    UsedColorArray:NiagaraDataInterfaceArrayInt32=$System.UsedColor
   ) {module=SetColorIndex}
   /SetVariables(Particles.CustomMeshIndex:NiagaraInt32=&MakeCustomIntFromBool(
     Bool=$User.TriggerKill
     True INT=&Add_Integer(
      A=&Multiply_Int(A=$Emitter.FramePerShape, B=$Emitter.ShapeNumber)
      B=&ModuloInt(
       A=&MakeIntFromFloat(Float=&Multiply_Float(A=$Engine.Time, B=8))
       B=$Emitter.FramePerShape
      )
     )
     False INT=&ModuloInt(
      A=&Add_Integer(
       A=&ModuloInt(
        A=&MakeIntFromFloat(Float=&Multiply_Float(A=$Engine.Time, B=8))
        B=$Emitter.FramePerShape
       )
       B=&Multiply_Int(
        A=&ModuloInt(
         A=&Add_Integer(A=$Particles.ShapeIndex, B=&ReturnExecIndex())
         B=$Emitter.ShapeNumber
        )
        B=$Emitter.FramePerShape
       )
      )
      B=&Multiply_Int(A=$Emitter.FramePerShape, B=$Emitter.ShapeNumber)
     )
    )) {isSetParameters=true, module=SetVariables_#19}
   /SetVariables(Particles.MeshOrientation=&MultiplyQuaternion(Quaternion A=$Particles.SkeletalMeshLocation.SampledRotation, Quaternion B=<0 0 0 1>)) {isSetParameters=true}
   &ScaleColor(
    Color Value To Scale=$Particles.Color
    Scale RGB=&Multiply_VectorByFloat(Vector=<1 1 1>, Float=1)
    ScaleA=true
    ScaleRGB=true
   ) {module=ScaleColor001}
   &Color(Color=&MakeCustomLinearColorFromBool(
     Bool=$User.TriggerKill
     True Color=&SelectLinearColorFromArray(
      Array Sampling Mode=::NewEnumerator1
      Direct Array Index=$Particles.ColorIndex
      Direct Array Mode=::NewEnumerator1
      Color Selection Array=$System.ColorArrayCurrentVariant
      Random Seed=?
     )
     False Color=&SelectLinearColorFromArray(
      Array Sampling Mode=::NewEnumerator1
      Direct Array Index=$Particles.ColorIndex
      Direct Array Mode=::NewEnumerator1
      Color Selection Array=$System.ColorArrayCurrentVariant
      Random Seed=?
     )
    )) {module=Color001}
   /SetVariables(Particles.Color=&MultiplyLinearColor(
     Linear Color=$Particles.Color
     Scale Factor=&MakeVector4(X=$System.TimeOfDayBrightnessMultiplier, Y=$System.TimeOfDayBrightnessMultiplier, Z=$System.TimeOfDayBrightnessMultiplier, W=1)
    )) {isSetParameters=true}
   &ScaleMeshSize(Scale Factor=&VectorFromFloat(Value=&MakeCustomFloatFromBool(Bool=$User.bisFootHidden, True Float=0, False Float=1)))
  ]
  rendererClasses [&NiagaraMeshRendererProperties]
  renderers [
   {rendererClass=&NiagaraMeshRendererProperties, rendererIndex=0}
   {rendererClass=&NiagaraMeshRendererProperties, rendererIndex=1}
  ]
 }
}
summary {
 systemName=NS_Skin_Dazzle_Idle
 emitters [
  {emitterName=NE_Squiggles, rendererClasses [&NiagaraMeshRendererProperties]}
  {emitterName=NE_Squiggles_Foot, rendererClasses [&NiagaraMeshRendererProperties]}
 ]
 userVariables [
  User.bisFootHidden {type=NiagaraBool}
  User.bIsFrontEnd {type=NiagaraBool}
  User.bIsFrontEndPreview {type=NiagaraBool}
  User.FlickerSpeed {type=NiagaraFloat, defaultValue=0.5, description=""}
  User.IsSkydiving {type=NiagaraBool}
  User.ShapeSwitchSpeed {type=NiagaraFloat, defaultValue=0.5, description=""}
  User.SkeletalMesh {
   type=NiagaraDataInterfaceSkeletalMesh
   defaultValue {
    struct=NiagaraExt_VariableValue_DataInterface
    value {dataInterface=&self:NiagaraDataInterfaceSkeletalMesh_0}
   }
  }
  User.TriggerKill {type=NiagaraBool}
  User.VariantIndex {type=NiagaraFloat, defaultValue=0, description=""}
 ]
}
