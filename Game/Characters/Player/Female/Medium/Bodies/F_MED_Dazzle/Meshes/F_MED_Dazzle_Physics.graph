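# Physics asset for the F_MED_Dazzle female medium player body, built around a dangling necklace on the upper chest.
# A three-link chain of capsule bodies hangs from master_necklace_jnt, each link constrained to the previous one with narrow, asymmetric swing limits and a zero twist limit.
# The master joint attaches to spine_05 and spine_05 to spine_04 with free angular motion; each spine bone carries two capsules forming the torso shape.
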PhysicsAsset /Game/Characters/Player/Female/Medium/Bodies/F_MED_Dazzle/Meshes/F_MED_Dazzle_Physics
bodies {
 dyn_necklace_1_jnt [{shapeType=Capsule, length=5, radius=5, shapeName=dyn_necklace_1_jnt_capsule, center=<4.1921 4.0554 0>}]
 dyn_necklace_2_jnt [{shapeType=Capsule, length=5, radius=5, shapeName=dyn_necklace_2_jnt_capsule, center=<4.1921 4.9134 0>}]
 dyn_necklace_3_jnt [{shapeType=Capsule, length=5, radius=5, shapeName=dyn_necklace_3_jnt_capsule, center=<4.1921 4.9014 0>}]
 master_necklace_jnt [{shapeType=Capsule, length=1, radius=1, shapeName=master_necklace_jnt_capsule}]
 spine_04 [
  {shapeType=Capsule, length=13.75, radius=8.5, shapeName=spine_05_capsule, center=<0 -0.0911 0>}
  {shapeType=Capsule, length=24, radius=6, shapeName=spine_05_capsule, center=<0.2503 5.8917 0>}
 ]
 spine_05 [
  {shapeType=Capsule, length=13.75, radius=8.5, shapeName=spine_05_capsule, center=<0 -0.0911 0>}
  {shapeType=Capsule, length=24, radius=6, shapeName=spine_05_capsule, center=<-3.9189 4.3297 0>}
 ]
}
constraints [
 {bone1Name=dyn_necklace_1_jnt, bone2Name=master_necklace_jnt, swing1LimitDegrees=25, swing2LimitDegrees=15, twistLimitDegrees=0}
 {bone1Name=dyn_necklace_2_jnt, bone2Name=dyn_necklace_1_jnt, swing1LimitDegrees=25, swing2LimitDegrees=15, twistLimitDegrees=0}
 {bone1Name=dyn_necklace_3_jnt, bone2Name=dyn_necklace_2_jnt, swing1LimitDegrees=25, swing2LimitDegrees=15, twistLimitDegrees=0}
 {
  bone1Name=master_necklace_jnt
  bone2Name=spine_05
  swing1LimitDegrees=45
  swing1Motion=Free
  swing2LimitDegrees=45
  swing2Motion=Free
  twistLimitDegrees=45
  twistMotion=Free
 }
 {
  bone1Name=spine_05
  bone2Name=spine_04
  swing1LimitDegrees=45
  swing1Motion=Free
  swing2LimitDegrees=45
  swing2Motion=Free
  twistLimitDegrees=45
  twistMotion=Free
 }
]
