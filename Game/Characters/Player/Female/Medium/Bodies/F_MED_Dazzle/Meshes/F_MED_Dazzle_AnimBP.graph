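# Post-process style AnimBlueprint for the F_MED_Dazzle female player body, layering secondary motion and deformation fixes onto the pose copied from the parent mesh.
# A deform Control Rig feeds a weighted zipper constraint, AnimDynamics and RigidBody simulation, then bone tweaks.
# Float alphas set from pawn states like tactical sprint, emote and skydive gate glute corrections and a front-end zipper rotation.
# The event graph also drives material world position offset from both hands and upper arms.
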AnimBlueprint /Game/Characters/Player/Female/Medium/Bodies/F_MED_Dazzle/Meshes/F_MED_Dazzle_AnimBP
target_skeleton=&/F_MED_Dazzle_Skeleton
graphs [
 AnimGraph {nodes [
   n1 = Control Rig!(
    Source=Copy Pose From Mesh() {bUseAttachedParent=true}
    Alpha
    NotCrouchingBool:bool=Get NotCrouching() -> (NotCrouching:bool)
    TacticalSprintingJumping:double=Get Tactical SprintingJump() -> (Tactical SprintingJump:double)
   ) -> (Pose) {
    alphaBoolBlend {blendOption=Linear}
    controlRigAssetReference {blueprintRigClass=&/F_MED_DeformRig_Dazzle_ControlRig.F_MED_DeformRig_Dazzle_ControlRig_C}
   }
   n2 = "Constraint - dyn_zipper_jnt (PPPPPP)"!(
    ComponentPose=Local To Component(LocalPose=@n1)
    Alpha=1
    ConstraintWeights_0=0.0522
    ConstraintWeights_1:float=0.0866
    ConstraintWeights_2:float=0.6719
    ConstraintWeights_3:float=0.0148
    ConstraintWeights_4:float=0.1575
    ConstraintWeights_5:float=0.017
   ) -> (Pose) {
    alphaBoolBlend {blendOption=Linear}
    boneToModify {boneName=dyn_zipper_jnt}
    constraintSetup [
     {targetBone {boneName=neck_02}}
     {targetBone {boneName=neck_01}}
     {targetBone {boneName=spine_05}}
     {targetBone {boneName=clavicle_r}}
     {targetBone {boneName=spine_04}}
     {targetBone {boneName=spine_03}}
    ]
    constraintWeights [1,1,1,1,1,1]
   }
   n3 = Anim Dynamics - Bone: dyn_zipper_jnt(ComponentPose=@n2, Alpha=1, ExternalForce="0,0,0") {
    angularDampingOverride=0.997
    angularSpringConstant=12000
    bAngularSpring=true
    bOverrideAngularDamping=true
    bUseGravityOverride=true
    alphaBoolBlend {blendOption=Linear}
    boundBone {boneName=dyn_zipper_jnt}
    componentAppliedLinearAccClamp=<3000 3000 3000>
    componentLinearAccScale=<0.1 0.1 0.12>
    componentLinearVelScale=<0.1 0.1 0.12>
    gravityOverride {z=-980}
    physicsBodyDefinitions [{
      boundBone {boneName=dyn_zipper_jnt}
      boxExtents=<10 10 10>
      constraintSetup {angularLimitsMax=<0 0 20>, angularLimitsMin=<0 0 -20>, angularTarget=<0 0 0>}
      localJointOffset=<5 0 0>
     }]
    relativeSpaceBone {boneName=root}
   }
   n4 = RigidBody(ComponentPose=@n3) {
    bClampLinearTranslationLimitToRefPose=true
    bOverrideWorldGravity=true
    overridePhysicsAsset=&/F_MED_Dazzle_Physics
    simulationSpace=BaseBoneSpace
    alphaBoolBlend {blendOption=Linear}
    baseBoneRef {boneName=spine_05}
    componentAppliedLinearAccClamp=<3000 3000 3000>
    componentLinearAccScale=<0.3 0.2 0.2>
    componentLinearVelScale=<0.3 0.2 0.2>
    overrideWorldGravity {z=-980}
    simSpaceSettings {maxAngularAcceleration=1000, maxAngularVelocity=3, maxLinearAcceleration=250, maxLinearVelocity=500, velocityScaleZ=0, worldAlpha=1}
   }
   n5 = Get Is Tactical Sprinting and Jumping() -> (Is Tactical Sprinting and Jumping:double)
   n6 = "Constraint - deform_glute_r (P)"(ComponentPose=@n4, Alpha=@n5, ConstraintWeights_0=1) {
    alphaBoolBlend {blendOption=Linear}
    boneToModify {boneName=deform_glute_r}
    constraintSetup [{targetBone {boneName=thigh_r}}]
    constraintWeights [1]
   }
   n7 = "Constraint - deform_glute_l (P)"(ComponentPose=@n6, Alpha=@n5, ConstraintWeights_0=1) {
    alphaBoolBlend {blendOption=Linear}
    boneToModify {boneName=deform_glute_l}
    constraintSetup [{targetBone {boneName=thigh_l}}]
    constraintWeights [1]
   }
   n8 = "Transform (Modify) Bone - Bone: deform_glute_l"(ComponentPose=@n7, Alpha=@n5, Translation="-2,0,0", Rotation="0,0,0", Scale="1,1,1") {
    translationMode=BMM_Additive
    translationSpace=BCS_BoneSpace
    alphaBoolBlend {blendOption=Linear}
    boneToModify {boneName=deform_glute_l}
    translation {x=-2}
   }
   n9 = "Transform (Modify) Bone - Bone: deform_glute_r"(ComponentPose=@n8, Alpha=@n5, Translation="2,0,0", Rotation="0,0,0", Scale="1,1,1") {
    translationMode=BMM_Additive
    translationSpace=BCS_BoneSpace
    alphaBoolBlend {blendOption=Linear}
    boneToModify {boneName=deform_glute_r}
    translation {x=2}
   }
   n10 = Get Performing Emote() -> (Performing Emote:double)
   n11 = "Constraint - deform_glute_r (P)"(ComponentPose=@n9, Alpha=@n10, ConstraintWeights_0=1) {
    alphaBoolBlend {blendOption=Linear}
    alphaScaleBiasClamp {bInterpResult=true, interpSpeedDecreasing=5}
    boneToModify {boneName=deform_glute_r}
    constraintSetup [{targetBone {boneName=thigh_r}}]
    constraintWeights [1]
   }
   n12 = "Constraint - deform_glute_l (P)"(ComponentPose=@n11, Alpha=@n10, ConstraintWeights_0=1) {
    alphaBoolBlend {blendOption=Linear}
    alphaScaleBiasClamp {bInterpResult=true, interpSpeedDecreasing=5}
    boneToModify {boneName=deform_glute_l}
    constraintSetup [{targetBone {boneName=thigh_l}}]
    constraintWeights [1]
   }
   n13 = "Transform (Modify) Bone - Bone: deform_glute_l"(ComponentPose=@n12, Alpha=@n10, Translation="-2,0,0", Rotation="0,0,0", Scale="1,1,1") {
    translationMode=BMM_Additive
    translationSpace=BCS_BoneSpace
    alphaBoolBlend {blendOption=Linear}
    alphaScaleBiasClamp {bInterpResult=true, interpSpeedDecreasing=5}
    boneToModify {boneName=deform_glute_l}
    translation {x=-2}
   }
   n14 = "Transform (Modify) Bone - Bone: deform_glute_r"(ComponentPose=@n13, Alpha=@n10, Translation="2,0,0", Rotation="0,0,0", Scale="1,1,1") {
    translationMode=BMM_Additive
    translationSpace=BCS_BoneSpace
    alphaBoolBlend {blendOption=Linear}
    alphaScaleBiasClamp {bInterpResult=true, interpSpeedDecreasing=5}
    boneToModify {boneName=deform_glute_r}
    translation {x=2}
   }
   n15 = Get Skydiving() -> (Skydiving:double)
   n16 = "Constraint - deform_glute_r (P)"(ComponentPose=@n14, Alpha=@n15, ConstraintWeights_0=1) {
    alphaBoolBlend {blendOption=Linear}
    boneToModify {boneName=deform_glute_r}
    constraintSetup [{targetBone {boneName=thigh_r}}]
    constraintWeights [1]
   }
   n17 = "Constraint - deform_glute_l (P)"(ComponentPose=@n16, Alpha=@n15, ConstraintWeights_0=1) {
    alphaBoolBlend {blendOption=Linear}
    boneToModify {boneName=deform_glute_l}
    constraintSetup [{targetBone {boneName=thigh_l}}]
    constraintWeights [1]
   }
   n18 = "Transform (Modify) Bone - Bone: deform_glute_l"(ComponentPose=@n17, Alpha=@n15, Translation="-2,0,0", Rotation="0,0,0", Scale="1,1,1") {
    translationMode=BMM_Additive
    translationSpace=BCS_BoneSpace
    alphaBoolBlend {blendOption=Linear}
    boneToModify {boneName=deform_glute_l}
    translation {x=-2}
   }
   n19 = "Transform (Modify) Bone - Bone: deform_glute_r"(ComponentPose=@n18, Alpha=@n15, Translation="2,0,0", Rotation="0,0,0", Scale="1,1,1") {
    translationMode=BMM_Additive
    translationSpace=BCS_BoneSpace
    alphaBoolBlend {blendOption=Linear}
    boneToModify {boneName=deform_glute_r}
    translation {x=2}
   }
   n20 = Get Tactical Sprinting() -> (Tactical Sprinting:double)
   n21 = "Transform (Modify) Bone - Bone: deform_glute_r"(ComponentPose=@n19, Alpha=@n20, Translation="3,0,0", Rotation="0,0,0", Scale="1,1,1") {
    translationMode=BMM_Additive
    translationSpace=BCS_BoneSpace
    alphaBoolBlend {blendOption=Linear}
    boneToModify {boneName=deform_glute_r}
    translation {x=3}
   }
   n22 = "Transform (Modify) Bone - Bone: deform_glute_l"(ComponentPose=@n21, Alpha=@n20, Translation="-3,0,0", Rotation="0,0,0", Scale="1,1,1") {
    translationMode=BMM_Additive
    translationSpace=BCS_BoneSpace
    alphaBoolBlend {blendOption=Linear}
    boneToModify {boneName=deform_glute_l}
    translation {x=-3}
   }
   n23 = "Transform (Modify) Bone - Bone: master_zipper_jnt"(ComponentPose=@n22, Alpha=Get Front End() -> (Front End:double), Translation="0,0,0", Rotation="0,10,0", Scale="1,1,1") {
    rotationMode=BMM_Additive
    rotationSpace=BCS_BoneSpace
    alphaBoolBlend {blendOption=Linear}
    boneToModify {boneName=master_zipper_jnt}
    rotation {yaw=10}
   }
   Output Pose(Result=Component To Local(ComponentPose=@n23)) {layerGroup=None}
  ]}
 EventGraph {nodes [
   n1 = Event BlueprintUpdateAnimation()
   > Is Valid(InputObject=Get PlayerPawn() -> (PlayerPawn:Fort Player Pawn Athena Object Reference))
   > Set WPO Globals Function(
    Material=Get Mat Instance Dynamic()
    Overlay Material=Get Overlay Mat Instance Dynamic()
    Skeletal Mesh=Get SkelMesh()
    "Local Noise Transform  Location"=Get Local Noise Transform Location() -> (Local Noise Transform Location:Vector)
    Local Noise Offset=Get Local Noise Offset()
   )
   > Set WPO Regional Function!(
    execute
    Material:Material Instance Dynamic Object Reference=Get Mat Instance Dynamic()
    Overlay Material:Material Instance Dynamic Object Reference=Get Overlay Mat Instance Dynamic()
    Skeletal Mesh:Skeletal Mesh Component Object Reference=Get SkelMesh()
    Velocity Transform:Name=hand_r
    Velocity Interp Speed:double=Get Velocity Interp Speed()
    Collision Transform:Name=upperarm_r
    Collision Transform Axis:EAxis Enum
    Invert Collision Transform Axis:bool=false
    Delta Time X:double=@n1.DeltaTimeX
    Previous Velocity Transform Location:"Vector (by ref)"=Get R1 Velocity Transform Location()
    Previous Velocity Transform Velocity:"Vector (by ref)"=Get R1 Velocity Transform Velocity()
    Region:int=1
   ) -> (then)
   > Set WPO Regional Function(
    Material=Get Mat Instance Dynamic()
    Overlay Material=Get Overlay Mat Instance Dynamic()
    Skeletal Mesh=Get SkelMesh()
    Velocity Transform=hand_l
    Velocity Interp Speed=Get Velocity Interp Speed()
    Collision Transform=upperarm_l
    Invert Collision Transform Axis=true
    Delta Time X=@n1.DeltaTimeX
    Previous Velocity Transform Location=Get R2 Velocity Transform Location()
    Previous Velocity Transform Velocity=Get R2 Velocity Transform Velocity()
    Region=2
   )
   n2 = OR Boolean(A=Get bIsJumping(), B=Get bIsFalling())
   n3 = Get bIsTacticalSprinting()
   > Set Is Tactical Sprinting and Jumping(Is Tactical Sprinting and Jumping:double="To Float (Boolean)"(InBool=AND Boolean(A=@n3, B=@n2))) -> (Output_Get:double)
   > Set Tactical Sprinting(Tactical Sprinting:double="To Float (Boolean)"(InBool=@n3)) -> (Output_Get:double)
   > Set Skydiving(Skydiving:double="To Float (Boolean)"(InBool=Get bIsSkyDiving())) -> (Output_Get:double)
   > Set Front End(Front End:double="To Float (Boolean)"(InBool=Get bIsInFrontEnd())) -> (Output_Get:double)
   > Set Performing Emote(Performing Emote:double="To Float (Boolean)"(InBool=Get bIsPlayingEmote())) -> (Output_Get:double)
   > Set NotCrouching(NotCrouching:bool=NOT Boolean(A=Get bIsCrouching())) -> (Output_Get:bool)
   n4 = Get bIsJumping()
   n5 = Get bIsTacticalSprinting()
   > Set Tactical SprintingJump(Tactical SprintingJump:double="To Float (Boolean)"(InBool=AND Boolean(A=@n5, B=@n4))) -> (Output_Get:double)
   > n6 = Branch(Condition=Get bIsSkydiveFloating())
   > Set WPO Pho Wind Intensity(WPO Pho Wind Intensity=1)
   Set WPO Pho Wind Intensity(execute=@n6.else, WPO Pho Wind Intensity=0)
   Event BlueprintInitializeAnimation()
   > Set SkelMesh(SkelMesh=GetOwnerSkeletalMeshComponent())
   n7 = TryGetPawnOwner()
   > Is Valid(InputObject=@n7)
   > n8 = Cast To FortPlayerPawnAthena(Object=@n7)
   > Set PlayerPawn(PlayerPawn:Fort Player Pawn Athena Object Reference=@n8) -> (Output_Get:Fort Player Pawn Athena Object Reference)
   Event OnCharacterCustomizationLoaded()
   > n9 = Set SkelMesh(SkelMesh=GetPartSkeletalMeshComponent())
   > n10 = Get Dynamic Material Instance(SkeletalMeshComponent=@n9)
   > Set Mat Instance Dynamic(Mat Instance Dynamic=@n10)
   > n11 = GetOrCreateDynamicOverlayMaterial(MeshComponent=@n9, MaterialName=None)
   > Set Overlay Mat Instance Dynamic(Overlay Mat Instance Dynamic=@n11)
  ]}
]
variables [
 Front End {type=number}
 Is Tactical Sprinting and Jumping {type=number}
 Local Noise Offset {type=Vector}
 Local Noise Transform Location {type=Vector}
 Mat Instance Dynamic {type=MaterialInstanceDynamic}
 NotCrouching {type=boolean}
 Overlay Mat Instance Dynamic {type=MaterialInstanceDynamic}
 Performing Emote {type=number}
 PlayerPawn {type=FortPlayerPawnAthena}
 R1 Velocity Transform Location {type=Vector}
 R1 Velocity Transform Velocity {type=Vector}
 R2 Velocity Transform Location {type=Vector}
 R2 Velocity Transform Velocity {type=Vector}
 SkelMesh {type=SkeletalMeshComponent}
 Skydiving {type=number}
 Tactical Sprinting {type=number}
 Tactical SprintingJump {type=number}
 Velocity Interp Speed {type=number, default=7}
 WPO Pho Wind Intensity {type=number}
]
